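# Translucent unlit, two-sided material for the Madame Moth pickaxe flame trail, built for Niagara ribbons and mesh particles.
# Opacity erodes a panning, UV-warped fire texture with noise, then applies edge masks, caustic-noise dissolve through HeightLerp and depth fade.
# Emissive multiplies a curve-atlas gradient along U by panning noise, particle color and a time-of-day tint.
# Dynamic parameters drive scroll speed, dissolve and SplineThicken width, which tapers along U.
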Material /Game/Weapons/FORT_Melee/Pickaxe_MadameMoth/FX/Material/M_Pickaxe_MadameMoth_Flame_Trail
graph {
 properties {
  bUsedWithNiagaraMeshParticles=true
  bUsedWithNiagaraRibbons=true
  blendMode=BLEND_Translucent
  pixelDepthOffsetMode=PDOM_Legacy
  shadingModel=MSM_Unlit
  twoSided=true
  usedShadingModels=Unlit
 }
 nodes [
  n1 = TextureSample(UVs=Panner(Coordinate=TextureCoordinate() {uTiling=3, vTiling=2}, Time=DynamicParameter() {paramNames [WPO, Speed, WPO Disturbance, Dissolve]}.Speed) {speedX=-0.5}) {samplerType=SAMPLERTYPE_Grayscale, texture=&LowResBlurredNoise}
  n2 = Multiply(A=TextureCoordinate(), B=ComponentMask(None=VectorParameter() {parameterName=UV Scale, defaultValue=<rgba 0.8 1 0 0>}.RGB) {g=true, r=true})
  n3 = AppendVector(A=Power(Base=ComponentMask(None=@n2) {r=true}) {constExponent=0.7}, B=ComponentMask(None=@n2) {g=true})
  n4 = DynamicParameter() {paramNames [WPO, Speed, WPO Disturbance, Dissolve]}
  n5 = Multiply(
   A=TextureSampleParameter2D(UVs=Panner(Coordinate=@n3, Time=@n4.Speed) {speedX=-0.3}) {parameterName=Tex_Default, texture=/Game/Weapons/FORT_Melee/Pickaxe_MadameMoth/FX/T_MadameMoth_Fire_Trail01}.B
   B=Saturate(None=LinearInterpolate(Alpha=@n1.R) {constB=1.5})
  )
  n6 = ScalarParameter() {defaultValue=1, parameterName=Lerp Center Dissolve}
  n7 = Multiply(A=TextureCoordinate(), B=ComponentMask(None=VectorParameter() {parameterName=UV Scale, defaultValue=<rgba 0.8 1 0 0>}.RGB) {g=true, r=true})
  n8 = AppendVector(A=Power(Base=ComponentMask(None=@n7) {r=true}) {constExponent=0.7}, B=ComponentMask(None=@n7) {g=true})
  n9 = DynamicParameter() {paramNames [WPO, Speed, WPO Disturbance, Dissolve]}
  n10 = ChannelMaskParameter(None=TextureSampleParameter2D(UVs=Panner(Coordinate=@n8, Time=@n9.Speed) {speedX=-0.3}) {parameterName=Tex_Default, texture=&T_MadameMoth_Fire_Trail01}.RGBA) {group=Tex, parameterName=Mask}
  n11 = ScalarParameter() {defaultValue=1, group=Tex, parameterName=Tex alpha fix}
  n12 = OneMinus(None=ComponentMask(None=TextureCoordinate()) {r=true})
  n13 = Multiply(
   A=Subtract(A=Power(Base=@n10.R, Exp=@n11) {constExponent=2.2}, B=LinearInterpolate(B=@n5, Alpha=@n6))
   B=Saturate(None=Multiply(A=@n12) {constB=5})
  )
  n14 = ParticleColor()
  n15 = Panner(Coordinate=TextureCoordinate() {uTiling=2.1, vTiling=0.9}) {speedX=-0.5}
  n16 = TextureCoordinate()
  n17 = LinearInterpolate(
   A=ScalarParameter() {parameterName=Dissolve mask Min}
   B=ScalarParameter() {defaultValue=0.5, parameterName=Dissolve mask Max}
   Alpha=Power(Base=ComponentMask(None=@n16) {r=true}) {constExponent=1.3}
  ) {constA=0.1, constB=2}
  n18 = LinearInterpolate(
   A=ScalarParameter() {defaultValue=-2, parameterName=Disappear min}
   B=ScalarParameter() {defaultValue=2, parameterName=Disappear max}
   Alpha=Power(
    Base=TextureSample(UVs=@n15) {samplerType=SAMPLERTYPE_LinearGrayscale, texture=/Game/Global/Textures/Default/Caustics/T_Nosie_Caustics_Grayscale}.R
    Exp=Saturate(None=@n17)
   )
  ) {constA=-2, constB=2}
  n19 = TextureCoordinate() {uTiling=2.1, vTiling=0.85}
  n20 = DynamicParameter() {paramNames [WPO, Speed, WPO Disturbance, Dissolve]}
  n21 = MaterialFunctionCall(
   Transition Phase=DynamicParameter() {paramNames [WPO, Speed, WPO Disturbance, Dissolve]}.Dissolve
   Height Texture=TextureSample(UVs=Panner(Coordinate=@n19, Time=@n20.Speed) {speedX=-0.5}) {samplerType=SAMPLERTYPE_LinearGrayscale, texture=&T_Nosie_Caustics_Grayscale}.R
  ) {materialFunction=&HeightLerp}
  n22 = Multiply(A=Multiply(A=Multiply(A=@n13, B=@n14.A), B=Saturate(None=@n18)), B=LinearInterpolate(Alpha=Multiply(A=@n21.Alpha)) {constA=-2, constB=2})
  n23 = ComponentMask(None=TextureCoordinate()) {r=true}
  n24 = LinearInterpolate(B=Saturate(None=Multiply(A=@n23) {constB=10}), Alpha=ScalarParameter() {defaultValue=1, parameterName=UseEdgeMask}) {constA=1}
  n25 = LinearInterpolate(A=TextureCoordinate(), B=TextureCoordinate() {coordinateIndex=1}, Alpha=ScalarParameter() {parameterName=Use mask UV1})
  n26 = OneMinus(None=ComponentMask(None=@n25) {r=true})
  n27 = Saturate(None=Multiply(A=Power(Base=@n26)) {constB=10})
  n28 = Power(Base=ComponentMask(None=@n25) {r=true}) {constExponent=1}
  n29 = LinearInterpolate(B=Saturate(None=Multiply(A=@n28) {constB=10}), Alpha=ScalarParameter() {defaultValue=1, parameterName=UseEdgeMask}) {constA=1}
  DepthFade(
   Opacity=Multiply(A=Multiply(A=@n22, B=@n24), B=Multiply(A=@n27, B=@n29))
   FadeDistance=ScalarParameter() {defaultValue=10, parameterName=FadeDistance}
  )
  DynamicParameter() {paramNames [WPO, Speed, WPO Disturbance, Dissolve]}
  n30 = Constant() {r=40}
  n31 = DynamicParameter() {paramNames [WPO, Speed, WPO Disturbance, Dissolve]}
  n32 = TextureCoordinate()
  n33 = Saturate(None=Multiply(A=ComponentMask(None=@n32) {r=true}) {constB=5})
  n34 = Constant() {r=40}
  MaterialFunctionCall(
   WidthBase=Multiply(A=Multiply(A=@n30, B=@n31.WPO), B=@n33)
   WidthTip=Multiply(A=Multiply(A=@n34, B=@n31.WPO), B=@n33)
   Expand U or V UV Channel=StaticBool()
  ) {materialFunction=&SplineThicken}
  n35 = Constant() {r=1}
  n36 = VectorParameter() {parameterName="Color_Night (V3)", defaultValue=<rgba 1.5 1.5 1.5 1>}
  n37 = TextureCoordinate() {uTiling=2}
  n38 = Constant2Vector() {r=0.2}
  n39 = TextureSample(UVs=Panner(Coordinate=Add(A=@n37, B=@n38), Time=DynamicParameter() {paramNames [WPO, Speed, WPO Disturbance, Dissolve]}.Speed) {speedX=-1}) {samplerType=SAMPLERTYPE_LinearGrayscale, texture=/Engine/Functions/Engine_MaterialFunctions02/ExampleContent/Textures/LowResBlurredNoise}
  n40 = Clamp(None=LinearInterpolate(Alpha=@n39.R) {constA=-0.8}, Min=ScalarParameter() {defaultValue=0.15, parameterName=Dark noise min}) {minDefault=0.15}
  n41 = Multiply(
   A=ScalarParameter() {defaultValue=1, parameterName=Add Noise}
   B=LinearInterpolate(
    B=DynamicParameter() {parameterIndex=2, defaultValue=<rgba 1 1 1 0>, paramNames [Add Noise Lerp, none, none, none]}.Add Noise Lerp
    Alpha=ScalarParameter() {parameterName=Lerp Dy Noise}
   ) {constA=1}
  )
  n42 = CurveAtlasRowParameter(CurveTime=ComponentMask(None=TextureCoordinate()) {r=true}) {
   atlas=/Game/Athena/LinearColorCurves/CAtlas_Contrail_MadameMoth_Color
   curve=/Game/Athena/LinearColorCurves/CV_Contrail_MadameMoth_Color1
   parameterName=Color
  }
  n43 = LinearInterpolate(
   B=Multiply(A=LinearInterpolate(B=@n40, Alpha=@n41) {constA=1}, B=@n42)
   Alpha=ScalarParameter() {defaultValue=1, parameterName=Use Color Curve}
  ) {constA=1}
  MaterialFunctionCall(Emissive=Multiply(
    A=MaterialFunctionCall(Color_Morning=@n35, Color_Day=@n35, Color_Evening=@n35, Color_Night=@n36.RGB) {materialFunction=&Color_Time_of_Day}.Result
    B=Multiply(A=@n43, B=@n14.RGB)
   )) {materialFunction=&MF_FX_EmissiveAdjustments}
  ScalarParameter() {defaultValue=1, parameterName=Add Noise}
 ]
}
